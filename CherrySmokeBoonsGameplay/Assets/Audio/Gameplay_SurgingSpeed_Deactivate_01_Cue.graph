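# Sound cue for the deactivation sound of the Surging Speed boon in the CherrySmoke boons gameplay plugin.
# A single wave player feeds the root directly, playing one deactivate wave with no randomization or modulation.
# It uses a close-range Athena weapon attenuation, the loot-collect sound class and reduced volume.
# A concurrency override limits it to one playing instance at a time.
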SoundCue /CherrySmokeBoonsGameplay/Assets/Audio/Gameplay_SurgingSpeed_Deactivate_01_Cue
root=WavePlayer_0
nodes [{id=WavePlayer_0, properties {soundWaveAssetPtr=&/Gameplay_SurgingSpeed_Deactivate_01}}]
properties {
 attenuationSettings=/Game/Sounds/Attenuation/Fort_Weapon_AB_Close_Athena_Att
 bOverrideConcurrency=true
 maxDistance=4500
 soundClassObject=&Fort_Loot_Collect
 volumeMultiplier=0.3
 concurrencyOverrides {maxCount=1}
}
