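# Sound cue for the Surging Speed boon in CherrySmokeBoons gameplay, layering an activation one-shot with looping speed and wind beds.
# A mixer plays the activate sound immediately while both loops start after a short delay and fade in through volume envelopers.
# Modulators randomize loop volume and the wind layer's pitch.
# Close-weapon attenuation and a single-instance concurrency override govern playback.
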SoundCue /CherrySmokeBoonsGameplay/Assets/Audio/Gameplay_SurgingSpeed_Loop_01_Cue
root=Mixer_0
nodes [
 {id=Delay_0, children [WavePlayer_1], properties {delayMax=0.1, delayMin=0.1}}
 {id=Delay_1, children [WavePlayer_2], properties {delayMax=0.1, delayMin=0.1}}
 {
  id=Enveloper_0
  children [Delay_1]
  properties {pitchCurve {editorCurveData {keys [{time=0}]}}, volumeCurve {editorCurveData {keys [<at 0 0>,<at 0.5 1>]}}}
 }
 {
  id=Enveloper_1
  children [Delay_0]
  properties {pitchCurve {editorCurveData {keys [{time=0}]}}, volumeCurve {editorCurveData {keys [<at 0 0>,<at 0.5 1>]}}}
 }
 {id=Mixer_0, children [WavePlayer_0,Modulator_0,Modulator_1], properties {inputVolume [1,1,1]}}
 {id=Modulator_0, children [Enveloper_1], properties {volumeMax=0.8, volumeMin=0.7}}
 {id=Modulator_1, children [Enveloper_0], properties {pitchMax=1.1, pitchMin=0.9, volumeMax=1.15, volumeMin=1.05}}
 {id=WavePlayer_0, properties {soundWaveAssetPtr=&/Gameplay_SurgingSpeed_Activate_01}}
 {id=WavePlayer_1, properties {bLooping=true, soundWaveAssetPtr=&/Gameplay_SurgingSpeed_Loop_01}}
 {id=WavePlayer_2, properties {bLooping=true, soundWaveAssetPtr=&/CherrySmokeBoons_wind_A}}
]
properties {
 attenuationSettings=/Game/Sounds/Attenuation/Fort_Weapon_AB_Close_Athena_Att
 bOverrideConcurrency=true
 maxDistance=4500
 soundClassObject=&Fort_Loot_Collect
 volumeMultiplier=0.35
 concurrencyOverrides {maxCount=1}
}
