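# Sound cue for the Surging Speed pickup in CherrySmokeBoons gameplay, routed to the Fort_Loot_Collect sound class.
# A mixer layers the default CherrySmoke medallion pickup wave, with slight random pitch, under a Surging Speed wave that starts after a short fixed delay.
# The Surging Speed layer varies slightly in volume only.
# A concurrency override limits the cue to one playing instance.
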SoundCue /CherrySmokeBoonsGameplay/Assets/Audio/Gameplay_SurgingSpeed_Pickup_01_Cue
root=Mixer_0
nodes [
 {id=Delay_0, children [Modulator_1], properties {delayMax=0.125, delayMin=0.125}}
 {id=Mixer_0, children [Modulator_0,Delay_0], properties {inputVolume [1,1]}}
 {id=Modulator_0, children [WavePlayer_3], properties {pitchMax=1.02, pitchMin=0.98}}
 {id=Modulator_1, children [WavePlayer_0], properties {pitchMax=1, pitchMin=1, volumeMax=0.95, volumeMin=0.85}}
 {id=WavePlayer_0, properties {soundWaveAssetPtr=&/Gameplay_SurgingSpeed_Pickup_01}}
 {id=WavePlayer_3, properties {soundWaveAssetPtr=&/CherrySmoke_Default_Medallion_PickUp_01}}
]
properties {
 bOverrideConcurrency=true
 maxDistance=2097152
 soundClassObject=&Fort_Loot_Collect
 volumeMultiplier=0.8
 concurrencyOverrides {maxCount=1}
}
