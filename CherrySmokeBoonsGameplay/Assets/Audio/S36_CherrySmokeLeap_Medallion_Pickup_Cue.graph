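# Sound cue for picking up the Cherry Smoke Leap medallion, routed to the Fort_Loot_Collect sound class.
# A Mixer layers the default medallion pickup sound with the Kor medallion music stinger, boosting both inputs.
# The pickup passes through a Modulator with pitch pinned at normal, and the stinger through a Delay left at default settings.
# A concurrency override limits the cue to one instance at a time.
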SoundCue /CherrySmokeBoonsGameplay/Assets/Audio/S36_CherrySmokeLeap_Medallion_Pickup_Cue
root=Mixer_0
nodes [
 {id=Delay_0, children [WavePlayer_1]}
 {id=Mixer_0, children [Modulator_0,Delay_0], properties {inputVolume [1.6,1.45]}}
 {id=Modulator_0, children [WavePlayer_3], properties {pitchMax=1, pitchMin=1}}
 {id=WavePlayer_1, properties {soundWaveAssetPtr=&/MX_S36_Medallion_Kor_STNG}}
 {id=WavePlayer_3, properties {soundWaveAssetPtr=&/CherrySmoke_Default_Medallion_PickUp_01}}
]
properties {
 bOverrideConcurrency=true
 maxDistance=2097152
 soundClassObject=&Fort_Loot_Collect
 volumeMultiplier=0.7692
 concurrencyOverrides {maxCount=1}
}
