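# Sound cue for consuming a CherrySmokeBoons boon, apparently the takedown medallion pickup, in the loot-collect sound class.
# A mixer layers the default takedown medallion pickup sound at reduced volume under the CherrySmokeBoons medallion sound at full volume.
# A modulator above the mixer adds slight random pitch variation to the combined result.
# A concurrency override limits it to one playing instance.
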SoundCue /CherrySmokeBoonsGameplay/Assets/Audio/SC_CherrySmokeBoons_Consume
root=Modulator_0
nodes [
 {id=Mixer_0, children [WavePlayer_4,WavePlayer_0], properties {inputVolume [0.4,1]}}
 {id=Modulator_0, children [Mixer_0], properties {pitchMax=1.03, pitchMin=0.97}}
 {id=WavePlayer_0, properties {soundWaveAssetPtr=&/CherrySmokeBoons_TakeDown_Medallion}}
 {id=WavePlayer_4, properties {soundWaveAssetPtr=&/Default_TakedownMedallions_Pickup_SFX}}
]
properties {
 bOverrideConcurrency=true
 maxDistance=2097152
 soundClassObject=&Fort_Loot_Collect
 volumeMultiplier=1
 concurrencyOverrides {maxCount=1}
}
